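# Niagara system for the CitrusSpoon player emote, adding a red eye glow and smoke puffs.
# The Empty emitter keeps a persistent glow sprite on a skeletal mesh socket, updated every frame.
# FeetSmoke fires three staggered bursts on Claws_SKM, using spawn groups and partitioning to pick bone or socket pairs, with front-end-dependent tint.
# User.LightOpacity scales glow and smoke alpha, plus the disabled Light emitter's color.
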NiagaraSystem /BRCosmetics/Animation/Game/MainPlayer/Emotes/CitrusSpoon/FX/NS_Emote_CitrusSpoon_EyeGlow
dependencies {
 usedDataInterfaces [
  &NiagaraDataInterfaceArrayInt32
  &NiagaraDataInterfaceCurve
  &NiagaraDataInterfaceColorCurve
  &NiagaraDataInterfaceSkeletalMesh
 ]
 usedDynamicInputs []
 usedModules [
  &self:SystemSpawnScript.SystemScriptSource.SystemScriptGraph.NiagaraNodeAssignment_0.SetVariables_#1
  &self:SystemSpawnScript.SystemScriptSource.SystemScriptGraph.NiagaraNodeAssignment_1.SetVariables_#2
  &SystemState
 ]
}
emitters {
 Empty {
  emitterUpdateScript [
   &EmitterState(Life Cycle Mode=::NewEnumerator1, Loop Behavior=::NewEnumerator1, Loop Duration=1, UseLoopDelay=false)
   &SpawnBurst_Instantaneous(Use Spawn Probability=false, UseLoopCountLimit=false)
  ]
  particleSpawnScript [
   &InitializeParticle(
    Color=<rgba 0.8542 0.0057 0 0.5>
    Lifetime=0.6
    Sprite Size Mode=::NewEnumerator3
    Uniform Sprite Size=40
    UsePositionOffset=false
    Write Material Random=true
   )
   &DynamicMaterialParameters(
    Index 0 Param 1=1
    Module.Index0Param0Write=true
    Module.Index0Param1Write=true
    Module.Index0Param2Write=true
    Module.Index0Param3Write=true
   ) {enabled=false}
  ]
  particleUpdateScript [
   &ParticleState(Kill Particles When Lifetime Has Elapsed=false)
   &SkeletalMeshLocation(
    MaskSpawnGroup=false
    Mesh Sampling Type=::NewEnumerator5
    Orientation Sampling=::NewEnumerator1
    Skeletal Mesh=$System.Skeletal Mesh
    Socket Sampling Mode=::NewEnumerator5
   )
   &SpriteFacingAndAlignment(
    Alignment Coordinate Space=::Local
    Sprite Alignment=<0 0 -1>
    Sprite Facing=&RotateVectorByQuaternion(Vector To Rotate=<0 0 0>, Quaternion=$Particles.SkeletalMeshLocation.SampledRotation)
    WriteAlignment=true
    WriteFacing=false
   )
   /SetVariables(Particles.GlowCameraOffset=5, Particles.GlowSize=&Vector2DFromFloat(Value=20)) {isSetParameters=true}
   &ScaleColor(Scale Alpha=$User.LightOpacity, ScaleA=true, ScaleRGB=true)
  ]
  rendererClasses [&NiagaraSpriteRendererProperties]
  renderers [
   {rendererClass=&NiagaraSpriteRendererProperties}
   {rendererClass=&NiagaraSpriteRendererProperties, rendererIndex=1}
  ]
 }
 FeetSmoke {
  emitterUpdateScript [
   &EmitterState(Inactive Response=::NewEnumerator1, Life Cycle Mode=::NewEnumerator1, Loop Duration=0.36, UseLoopDelay=false)
   &SpawnBurst_Instantaneous(Spawn Count=4, Use Spawn Probability=false, UseLoopCountLimit=false)
   &SpawnBurst_Instantaneous(Spawn Count=4, Spawn Group=1, Spawn Time=0.08, Use Spawn Probability=false, UseLoopCountLimit=false) {module=SpawnBurst_Instantaneous001}
   &SpawnBurst_Instantaneous(Spawn Count=4, Spawn Group=2, Spawn Time=0.16, Use Spawn Probability=false, UseLoopCountLimit=false) {module=SpawnBurst_Instantaneous002}
  ]
  particleSpawnScript [
   &InitializeParticle(
    Color=&MakeCustomLinearColorFromBool(Bool=$User.bIsFrontEnd, True Color=<rgba 0.3125 0.2026 0.168 0.2>, False Color=<rgba 0.3125 0.2688 0.2466 0.2>)
    Lifetime Max=1.1
    Lifetime Min=0.6
    Lifetime Mode=::NewEnumerator1
    Sprite Size Mode=::NewEnumerator1
    Uniform Sprite Size Max=45
    Uniform Sprite Size Min=30
    UsePositionOffset=false
    Write Material Random=true
   )
   &PartitionParticles()
   &SkeletalMeshLocation(
    Bone / Socket Index=&MaskIntBySpawnGroup(
     Masked Int=&SelectIntFromArray(Direct Array Index=$Particles.PartitionParticles.Partition, Int Selection Array=DataInterface() {propertyValues {IntData [0,3]}})
     Passthrough Int=&MaskIntBySpawnGroup(
      Masked Int=&SelectIntFromArray(Direct Array Index=$Particles.PartitionParticles.Partition, Int Selection Array=DataInterface() {propertyValues {IntData [1,4]}})
      Passthrough Int=&SelectIntFromArray(Direct Array Index=$Particles.PartitionParticles.Partition, Int Selection Array=DataInterface() {propertyValues {IntData [2,5]}})
      Spawn Group Mask=1
     )
     Spawn Group Mask=0
    )
    Bone Sampling Mode=::NewEnumerator4
    MaskSpawnGroup=false
    Sampled Position Offset=<0 7.5 0>
    Sampled Position Offset Coordinate Space:ENiagaraSkelMeshTransforms=::NewEnumerator2
    Skeletal Mesh=$System.Claws_SKM
   )
   &ShapeLocation(MaskSpawnGroup=false, Override Seed=false, Sphere Radius=6, UseOwnerScale=true)
   &AddVelocity(Rotation Coordinate Space=::World, Velocity=&RandomRangeVector(Minimum=<25 25 0>, Maximum=<-25 -25 1>))
  ]
  particleUpdateScript [
   &GravityForce(Gravity=<0 0 150>) {enabled=false}
   &SolveForcesAndVelocity(Clamp Velocity=false, Limit Acceleration=false)
   &ParticleState()
   &ScaleSpriteSize(Uniform Curve Sprite Scale=DataInterface() {propertyValues {Curve {keys [
        {interpMode=RCIM_Cubic, leaveTangent=1.4818, tangentMode=RCTM_Break, time=0, value=0.35}
        {arriveTangent=-0.079, interpMode=RCIM_Cubic, tangentMode=RCTM_Break, time=1, value=1}
       ]}}})
   &ScaleColor(
    Scale Alpha=&FloatFromCurve(
     FloatCurve=DataInterface() {propertyValues {Curve {keys [<at 0 0.25>, {interpMode=RCIM_Cubic, leaveTangent=-4.4113, tangentMode=RCTM_Break, time=0.2, value=1}, <at 1 0>]}}}
     Scale Curve=&Multiply_Float(A=1, B=$User.LightOpacity)
    )
    ScaleA=true
    ScaleRGB=true
   )
   &DynamicMaterialParameters(
    Index 0 Param 1=&Add_Float(
     A=&RandomRangeFloat(Minimum=0, Maximum=1)
     B=&FloatFromCurve(
      FloatCurve=DataInterface() {propertyValues {Curve {keys [
          <at 0 0 RCIM_Cubic>
          {arriveTangent=1.5947, interpMode=RCIM_Cubic, leaveTangent=1.5947, tangentMode=RCTM_User, time=0.2557, value=0.5257}
          <at 1 1 RCIM_Cubic>
         ]}}}
      CurveIndex=$Particles.Age
      Scale Curve=0.25
     )
    )
    Index 0 Param 4=1
    Module.Param0WriteEnabled=true
    Module.Param1WriteEnabled=true
    Module.Param2WriteEnabled=true
    Module.Param3WriteEnabled=true
   ) {module=DynamicMaterialParameters001}
   &CameraOffset(Camera Offset Amount=20)
  ]
  rendererClasses [&NiagaraSpriteRendererProperties]
  renderers [
   {rendererClass=&NiagaraSpriteRendererProperties}
   {rendererClass=&NiagaraSpriteRendererProperties, rendererIndex=1}
  ]
 }
 Light {
  bEnabled=false
  emitterUpdateScript [
   &EmitterState()
   &SpawnBurst_Instantaneous(Use Spawn Probability=false, UseLoopCountLimit=false)
  ]
  particleSpawnScript [&InitializeParticle(
    Color=<rgba 1 0.0237 0 1>
    Lifetime=0.6
    Sprite Rotation Angle=90
    Sprite Rotation Mode=::NewEnumerator0
    Sprite Size Mode=::NewEnumerator3
    Uniform Sprite Size=15
    UsePositionOffset=false
    Write Material Random=true
   )]
  particleUpdateScript [
   &ParticleState(Kill Particles When Lifetime Has Elapsed=false)
   &SkeletalMeshLocation(
    Bone / Socket Index=1
    MaskSpawnGroup=false
    Mesh Sampling Type=::NewEnumerator5
    Skeletal Mesh=$System.Skeletal Mesh
    Socket Sampling Mode=::NewEnumerator5
   )
   &FN_Light_Attributes(
    Color=<rgba 1 0 0.0308 1>
    Create Color Attribute=true
    Create Exponent Attribute=true
    Create Radius Attribute=true
    Exponent=2
    Radius=12
    Scale Color=&Multiply_Float(A=6, B=$User.LightOpacity)
   )
  ]
  rendererClasses [&NiagaraLightRendererProperties]
  renderers [{rendererClass=&NiagaraLightRendererProperties, rendererIndex=0}]
 }
}
summary {
 systemName=NS_Emote_CitrusSpoon_EyeGlow
 emitters [
  {emitterName=Empty, rendererClasses [&NiagaraSpriteRendererProperties]}
  {bEnabled=false, emitterName=Light, rendererClasses [&NiagaraLightRendererProperties]}
  {emitterName=FeetSmoke, rendererClasses [&NiagaraSpriteRendererProperties]}
 ]
 userVariables [
  User.bIsFrontEnd {type=NiagaraBool, defaultValue=true}
  User.LightOpacity {type=NiagaraFloat, defaultValue=1, description=""}
 ]
}
