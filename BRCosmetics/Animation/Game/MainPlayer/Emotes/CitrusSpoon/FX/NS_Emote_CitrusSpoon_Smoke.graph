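# Niagara system for the CitrusSpoon emote that kicks up smoke from a skeletal mesh, as large soft Dust sprites and small gravity-pulled Chunks.
# Both emitters spawn per unit of movement, adding a constant spawn rate only in the front end.
# Front-end modifiers and User.bIsFrontEnd switch velocity and tint between menu and gameplay.
# Sprites grow and fade over life, with randomised, age-driven dynamic material parameters and a camera offset.
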NiagaraSystem /BRCosmetics/Animation/Game/MainPlayer/Emotes/CitrusSpoon/FX/NS_Emote_CitrusSpoon_Smoke
dependencies {
 usedDataInterfaces [&NiagaraDataInterfaceCurve, &NiagaraDataInterfaceColorCurve, &NiagaraDataInterfaceSkeletalMesh]
 usedDynamicInputs []
 usedModules [
  &self:SystemSpawnScript.SystemScriptSource.SystemScriptGraph.NiagaraNodeAssignment_0.SetVariables_#1
  &SystemState
 ]
}
emitters {
 Chunks {
  emitterUpdateScript [
   &EmitterState()
   &SpawnRate(SpawnRate=&FN_FrontEndFloatModifier(Gameplay=0, Front End=4), Use Spawn Probability=false)
   &SpawnPerUnit(Module.Use Max Movement Threshold=true, Module.Use Movement Tolerance=false, Spawn Spacing=25, Use Spawn Probability=false)
  ]
  particleSpawnScript [
   &InitializeParticle(
    Color=&MakeCustomLinearColorFromBool(Bool=$User.bIsFrontEnd, True Color=<rgba 1 0.6277 0.4762 1>, False Color=<rgba 1.5 0.9415 0.7142 1>)
    Lifetime Max=0.6
    Lifetime Min=0.4
    Lifetime Mode=::NewEnumerator1
    Sprite Size Mode=::NewEnumerator1
    Uniform Sprite Size Min=7
    UsePositionOffset=false
    Write Material Random=true
   )
   &SkeletalMeshLocation(Bone Sampling Mode=::NewEnumerator4, MaskSpawnGroup=false, Skeletal Mesh=$System.Skeletal Mesh)
   &ShapeLocation(MaskSpawnGroup=false, Override Seed=false, Sphere Radius=30, UseOwnerScale=true)
   &AddVelocity(Velocity=&FN_FrontEndVector3Modifier(
     Gameplay=&RandomRangeVector(Minimum=<0 -25 55>, Maximum=<0 -45 150>)
     Front End=&RandomRangeVector(Minimum=<0 -150 60>, Maximum=<0 -200 200>)
    ))
  ]
  particleUpdateScript [
   &ParticleState()
   &ScaleSpriteSize(Uniform Curve Sprite Scale=DataInterface() {propertyValues {Curve {keys [
        {tangentMode=RCTM_User, time=0, value=0}
        {interpMode=RCIM_Cubic, leaveTangent=0.9898, tangentMode=RCTM_Break, time=0.2, value=0.75}
        {tangentMode=RCTM_User, time=1, value=1.25}
       ]}}})
   &GravityForce(Gravity=<0 0 -400>)
   &SolveForcesAndVelocity(Clamp Velocity=false, Limit Acceleration=false)
   &DynamicMaterialParameters(
    Index 0 Param 1=&Add_Float(
     A=&RandomRangeFloat(Minimum=0, Maximum=1)
     B=&FloatFromCurve(
      FloatCurve=DataInterface() {propertyValues {Curve {keys [
          <at 0 0 RCIM_Cubic>
          {arriveTangent=1.5947, interpMode=RCIM_Cubic, leaveTangent=1.5947, tangentMode=RCTM_User, time=0.2557, value=0.5257}
          <at 1 1 RCIM_Cubic>
         ]}}}
      CurveIndex=$Particles.Age
      Scale Curve=0.25
     )
    )
    Index 0 Param 4=1
    Module.Param0WriteEnabled=true
    Module.Param1WriteEnabled=true
    Module.Param2WriteEnabled=true
    Module.Param3WriteEnabled=true
   ) {module=DynamicMaterialParameters001}
   &ScaleColor(
    Linear Color Curve=DataInterface() {propertyValues {
      AlphaCurve {keys [<at 0.0008 0.4>,<at 0.1286 1>,<at 1 0>]}
      BlueCurve {keys [<at 0 1>,<at 1 1>]}
      GreenCurve {keys [<at 0 1>,<at 1 1>]}
      RedCurve {keys [<at 0 1>,<at 1 1>]}
     }}
    Scale Mode=::NewEnumerator2
   )
   &CameraOffset(Camera Offset Amount=10)
  ]
  rendererClasses [&NiagaraSpriteRendererProperties]
  renderers [{rendererClass=&NiagaraSpriteRendererProperties}]
 }
 Dust {
  emitterUpdateScript [
   &EmitterState()
   &SpawnRate(SpawnRate=&FN_FrontEndFloatModifier(Gameplay=0, Front End=5), Use Spawn Probability=false)
   &SpawnPerUnit(Module.Use Max Movement Threshold=true, Module.Use Movement Tolerance=false, Spawn Spacing=26, Use Spawn Probability=false)
  ]
  particleSpawnScript [
   &InitializeParticle(
    Color=&MakeCustomLinearColorFromBool(Bool=$User.bIsFrontEnd, True Color=<rgba 0.3125 0.2026 0.168 0.35>, False Color=<rgba 0.3125 0.2688 0.2466 0.35>)
    Lifetime Max=1.1
    Lifetime Min=0.75
    Lifetime Mode=::NewEnumerator1
    Sprite Size Mode=::NewEnumerator1
    Uniform Sprite Size Max=100
    Uniform Sprite Size Min=70
    UsePositionOffset=false
    Write Material Random=true
   )
   &SkeletalMeshLocation(Bone Sampling Mode=::NewEnumerator4, MaskSpawnGroup=false, Skeletal Mesh=$System.Skeletal Mesh)
   &AddVelocity(Velocity=&FN_FrontEndVector3Modifier(
     Gameplay=&RandomRangeVector(Minimum=<0 -15 5>, Maximum=<0 -30 15>)
     Front End=&RandomRangeVector(Minimum=<0 -150 10>, Maximum=<0 -200 30>)
    ))
  ]
  particleUpdateScript [
   &ParticleState()
   &ScaleSpriteSize(Uniform Curve Sprite Scale=DataInterface() {propertyValues {Curve {keys [
        {interpMode=RCIM_Cubic, leaveTangent=1.4818, tangentMode=RCTM_Break, time=0, value=0.35}
        {arriveTangent=-0.079, interpMode=RCIM_Cubic, tangentMode=RCTM_Break, time=1, value=1}
       ]}}})
   &SolveForcesAndVelocity(Clamp Velocity=false, Limit Acceleration=false)
   &DynamicMaterialParameters(
    Index 0 Param 1=&Add_Float(
     A=&RandomRangeFloat(Minimum=0, Maximum=1)
     B=&FloatFromCurve(
      FloatCurve=DataInterface() {propertyValues {Curve {keys [
          <at 0 0 RCIM_Cubic>
          {arriveTangent=1.5947, interpMode=RCIM_Cubic, leaveTangent=1.5947, tangentMode=RCTM_User, time=0.2557, value=0.5257}
          <at 1 1 RCIM_Cubic>
         ]}}}
      CurveIndex=$Particles.Age
      Scale Curve=0.25
     )
    )
    Index 0 Param 4=1
    Module.Param0WriteEnabled=true
    Module.Param1WriteEnabled=true
    Module.Param2WriteEnabled=true
    Module.Param3WriteEnabled=true
   ) {module=DynamicMaterialParameters001}
   &ScaleColor(
    Linear Color Curve=DataInterface() {propertyValues {
      AlphaCurve {keys [<at 0.0008 0.4>,<at 0.1286 1>,<at 0.6412 0.2>,<at 1 0>]}
      BlueCurve {keys [<at 0 1>,<at 1 1>]}
      GreenCurve {keys [<at 0 1>,<at 1 1>]}
      RedCurve {keys [<at 0 1>,<at 1 1>]}
     }}
    Scale Mode=::NewEnumerator2
   )
   &CameraOffset(Camera Offset Amount=10)
  ]
  rendererClasses [&NiagaraSpriteRendererProperties]
  renderers [{rendererClass=&NiagaraSpriteRendererProperties}]
 }
}
summary {
 systemName=NS_Emote_CitrusSpoon_Smoke
 emitters [
  {emitterName=Dust, rendererClasses [&NiagaraSpriteRendererProperties]}
  {emitterName=Chunks, rendererClasses [&NiagaraSpriteRendererProperties]}
 ]
 userVariables [
  User.bIsFrontEnd {type=NiagaraBool}
  User.bIsFrontEndPreview {type=NiagaraBool}
 ]
}
